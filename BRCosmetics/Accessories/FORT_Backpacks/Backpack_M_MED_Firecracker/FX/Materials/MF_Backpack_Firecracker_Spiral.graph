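# Material function for the Firecracker backpack FX that curls a mesh into an animated spiral through World Position Offset.
# A gradient along a tilted local length axis, plus time, sets the rotation angle around local Y at the object pivot.
# Before rotation, vertices collapse along that axis, shift sideways with skew and puff along pre-skinned normals.
# It also outputs matching rotated normal and length vectors, the gradient, puff vector and rotation speed.
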MaterialFunction /BRCosmetics/Accessories/FORT_Backpacks/Backpack_M_MED_Firecracker/FX/Materials/MF_Backpack_Firecracker_Spiral
graph {nodes [
  n1 = Constant3Vector() {desc=Rotation Axis, constant=<rgba 0 1 0 0>}
  n2 = Normalize(VectorInput=Transform(None=@n1) {transformSourceType=TRANSFORMSOURCE_Local})
  n3 = FunctionInput(Preview=WorldPosition().XYZ) {bUsePreviewValueAsDefault=true, description=Defaults to Absolute World Position, inputName=WorldPosition}
  n4 = Constant3Vector() {desc=Top-down Gradient, constant=<rgba -0.2582 0 0.966 0>}
  n5 = FunctionOutput(None=DotProduct(A=TransformPosition(None=@n3) {transformSourceType=TRANSFORMPOSSOURCE_World}, B=@n4)) {outputName=Length Gradient, sortPriority=4}
  n6 = ScalarParameter() {defaultValue=1.5, parameterName=Rotation Speed}
  n7 = Add(
   A=Multiply(A=@n5, B=ScalarParameter() {defaultValue=0.03, parameterName=RotationAmount})
   B=Multiply(A=Time(), B=FunctionOutput(None=@n6) {outputName=Rotation Speed})
  )
  n8 = Constant()
  n9 = FunctionInput(Preview=VertexNormalWS()) {bUsePreviewValueAsDefault=true, description=Defaults to VertexNormalWS, inputName=WorldSpace Normal}
  n10 = RotateAboutAxis(NormalizedRotationAxis=@n2, RotationAngle=@n7, PivotPoint=@n8, Position=@n9)
  FunctionOutput(None=Normalize(VectorInput=Add(A=@n10, B=@n9))) {outputName=Spiral WorldSpace Normal}
  n11 = FunctionOutput(None=Normalize(VectorInput=Transform(None=@n4) {transformSourceType=TRANSFORMSOURCE_Local})) {outputName=Length Vector, sortPriority=4}
  n12 = Multiply(A=@n5, B=@n11)
  n13 = CrossProduct(A=@n4, B=@n1)
  n14 = ScalarParameter() {defaultValue=-0.2367, parameterName=Skew}
  n15 = Add(A=Multiply(A=Multiply(A=@n13, B=@n14), B=@n5), B=Multiply(A=@n13, B=ScalarParameter() {defaultValue=11, parameterName=Side Pivot Offset}))
  n16 = Transform(None=PreSkinnedNormal()) {transformSourceType=TRANSFORMSOURCE_Local}
  n17 = Clamp(None=Subtract(A=@n5, B=ScalarParameter() {defaultValue=-2.0934, parameterName=Puff Gradient Shift})) {clampMode=CMODE_ClampMax, maxDefault=0}
  n18 = Add(
   A=Add(A=Multiply(A=@n12) {constB=-1}, B=Transform(None=@n15) {transformSourceType=TRANSFORMSOURCE_Local})
   B=Multiply(A=Multiply(A=@n16, B=@n17), B=ScalarParameter() {defaultValue=-0.2, parameterName=Puff})
  ) {constB=0}
  n19 = RotateAboutAxis(
   NormalizedRotationAxis=@n2
   RotationAngle=@n7
   PivotPoint=MaterialFunctionCall() {materialFunction=&ObjectPivotPoint}.Object Pivot Location
   Position=Add(A=@n18, B=@n3)
  )
  n20 = ScalarParameter() {defaultValue=-3, parameterName=Push Along Rotation Axis}
  FunctionOutput(None=Add(A=Add(A=@n18, B=@n19), B=Multiply(A=@n2, B=@n20))) {outputName=Spiral WorldPositionOffset, sortPriority=1}
  FunctionOutput(None=@n16) {outputName=Puff WorldSpace Vector, sortPriority=2}
  n21 = RotateAboutAxis(NormalizedRotationAxis=@n2, RotationAngle=@n7, PivotPoint=@n8, Position=@n11)
  FunctionOutput(None=Normalize(VectorInput=Add(A=@n21, B=@n11))) {outputName=Spiral WorldSpace Length Vector, sortPriority=2}
 ]}
