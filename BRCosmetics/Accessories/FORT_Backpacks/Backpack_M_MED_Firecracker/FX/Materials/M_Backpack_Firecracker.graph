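# Masked, lit material for the Firecracker backpack cosmetic, with a spiral-deformed wick and an orange glow.
# A spiral material function supplies world position offset, world-space normals and a length gradient.
# Per vertex, the thresholded gradient forms the wick opacity mask and, with a pre-skinned sphere mask for the bottom glow, passes through a vertex interpolator.
# Both masks drive an HDR emissive color through a shared FX adjustment function.
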Material /BRCosmetics/Accessories/FORT_Backpacks/Backpack_M_MED_Firecracker/FX/Materials/M_Backpack_Firecracker
graph {
 properties {
  bTangentSpaceNormal=false
  bUsedWithMeshDeformer=true
  bUsedWithSkeletalMesh=true
  blendMode=BLEND_Masked
  pixelDepthOffsetMode=PDOM_Legacy
  previewMesh=/BRCosmetics/Accessories/FORT_Backpacks/Backpack_M_MED_Firecracker/Meshes/M_MED_Firecracker_Pack
  usedShadingModels=Default Lit
 }
 nodes [
  n1 = VectorParameter() {parameterName=Color Emissive, defaultValue=<rgba 333 82.584 13.32 0.0005>}
  n2 = TextureSample() {
   samplerType=SAMPLERTYPE_Normal
   texture=/BRCosmetics/Accessories/FORT_Backpacks/Backpack_M_MED_Firecracker/Textures/T_M_MED_Firecracker_Backpack_N
  }
  n3 = MaterialFunctionCall!(WorldPosition, WorldSpace Normal=Transform(None=@n2.RGB)) {materialFunction=&/MF_Backpack_Firecracker_Spiral}
  n4 = ScalarParameter() {defaultValue=1.05, parameterName=Wick Opacity Threshold}
  n5 = AppendVector(
   A=Ceil(None=Add(A=@n3.Length Gradient, B=@n4))
   B=SphereMask(A=PreSkinnedPosition(), B=Constant3Vector() {constant=<rgba 2 -2 -2 0>}, Radius=Constant() {r=1.6})
  )
  n6 = VertexInterpolator(VS=AppendVector(A=@n5, B=ComponentMask(None=@n5) {r=true}))
  n7 = Power(
   Base=ComponentMask(None=@n6.PS) {desc=Glowy bottom, g=true}
   Exp=ScalarParameter() {defaultValue=3, parameterName=Bottom Light Falloff, sliderMax=9, sliderMin=1}
  )
  n8 = Divide(
   A=ComponentMask(None=@n6.PS) {b=true, desc=Emissive mask}
   B=ScalarParameter() {defaultValue=2, parameterName=Bottom Light Range, sliderMax=1}
  )
  n9 = Multiply(A=OneMinus(None=Saturate(None=@n8)), B=@n1.A)
  MaterialFunctionCall(Emissive=Multiply(A=@n1.RGB, B=Max(A=@n7, B=@n9))) {materialFunction=&MF_FX_EmissiveAdjustments}
  Reroute(None=@n3.Spiral WorldPositionOffset)
  Reroute(None=@n3.Spiral WorldSpace Normal)
  Reroute(None=ComponentMask(None=@n6.PS) {desc=Opacity Mask, r=true})
  TextureSample() {texture=/BRCosmetics/Accessories/FORT_Backpacks/Backpack_M_MED_Firecracker/Textures/T_M_MED_Firecracker_Backpack_D}
  TextureSample() {
   samplerType=SAMPLERTYPE_Masks
   texture=/BRCosmetics/Accessories/FORT_Backpacks/Backpack_M_MED_Firecracker/Textures/T_M_MED_Firecracker_Backpack_S
  }
  TextureSample() {
   samplerType=SAMPLERTYPE_Masks
   texture=/BRCosmetics/Accessories/FORT_Backpacks/Backpack_M_MED_Firecracker/Textures/T_M_MED_Firecracker_Backpack_M
  }
 ]
}
