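# Unlit modulate Niagara sprite material that fakes light cast by the Firecracker backpack onto nearby surfaces.
# A soft sphere mask orbits in UV space, timed by the spiral function's Rotation Speed output and shaped by a Color-alpha exponent.
# An inverted depth fade confines it near scene geometry, while a short second fade softens intersections.
# That mask blends from neutral white to the bright Color parameter.
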Material /BRCosmetics/Accessories/FORT_Backpacks/Backpack_M_MED_Firecracker/FX/Materials/M_Backpack_Firecracker_FakeLight
graph {
 properties {
  bUsedWithNiagaraSprites=true
  blendMode=BLEND_Modulate
  pixelDepthOffsetMode=PDOM_Legacy
  shadingModel=MSM_Unlit
  translucencyPass=MTP_BeforeDOF
  usedShadingModels=Unlit
 }
 nodes [
  n1 = VectorParameter() {parameterName=Color, defaultValue=<rgba 11 2.9865 0.77 3>}
  n2 = Constant2Vector() {g=0.666, r=0.4}
  n3 = Multiply(
   A=MaterialFunctionCall!(WorldPosition, WorldSpace Normal) {materialFunction=&/MF_Backpack_Firecracker_Spiral}.Rotation Speed
   B=Time()
  )
  n4 = Power(
   Base=SphereMask(A=TextureCoordinate(), B=Rotator(Coordinate=@n2, Time=@n3) {speed=-6.28}, Radius=Constant() {r=0.333}) {hardnessPercent=0}
   Exp=@n1.A
  ) {constExponent=3}
  n5 = OneMinus(None=DepthFade(FadeDistance=ScalarParameter() {defaultValue=33, parameterName=DepthFadeDistance}))
  LinearInterpolate(B=@n1.RGB, Alpha=DepthFade(Opacity=Multiply(A=@n4, B=@n5), FadeDistance=ScalarParameter() {defaultValue=3, parameterName=CloseDepthFade})) {constA=1}
 ]
}
