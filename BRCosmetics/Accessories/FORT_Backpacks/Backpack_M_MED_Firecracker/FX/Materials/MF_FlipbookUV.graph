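# MF_FlipbookUV is a material function, stored with the Firecracker backpack FX, that computes flipbook UVs for crossfading between frames on a grid.
# The fraction of the animation phase, game time by default, scales by the grid's frame count to give a looping frame position.
# The current and next frame indices become row and column offsets added to the scaled-down UVs.
# It outputs both UV sets plus the fractional frame position as blend alpha.
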MaterialFunction /BRCosmetics/Accessories/FORT_Backpacks/Backpack_M_MED_Firecracker/FX/Materials/MF_FlipbookUV
graph {nodes [
  n1 = FunctionInput(Preview=Time()) {bUsePreviewValueAsDefault=true, description=Defaults to game time, inputName=Animation Phase, inputType=FunctionInput_Scalar, sortPriority=2}
  n2 = FunctionInput() {
   bUsePreviewValueAsDefault=true
   description=Amount of frames in the X and Y axes
   inputName=XY Frames
   inputType=FunctionInput_Vector2
   sortPriority=1
   previewValue=<2 2 0 1>
  }
  n3 = ComponentMask(None=@n2) {r=true}
  n4 = ComponentMask(None=@n2) {g=true}
  n5 = Multiply(A=Frac(None=@n1), B=Multiply(A=@n3, B=@n4)) {desc="Frame ##, float"}
  n6 = Floor(None=@n5) {desc="Frame ##, int"}
  n7 = Divide(A=@n6, B=@n3) {desc=X}
  n8 = AppendVector(A=@n7, B=Divide(A=Floor(None=@n7), B=@n4) {desc=Y})
  n9 = Divide(
   A=FunctionInput(Preview=TextureCoordinate()) {bUsePreviewValueAsDefault=true, description="Defaults to TexCoord[0]", inputName=UV, inputType=FunctionInput_Vector2}
   B=@n2
  )
  FunctionOutput(None=Add(A=@n9, B=Frac(None=@n8) {desc=XY})) {description=Flipbook UV 1, outputName=UV 1}
  n10 = Divide(A=Add(A=@n6), B=@n3) {desc=X}
  n11 = AppendVector(A=@n10, B=Divide(A=Floor(None=@n10), B=@n4) {desc=Y})
  FunctionOutput(None=Add(A=@n9, B=Frac(None=@n11) {desc=XY})) {description=Flipbook UV 2, outputName=UV 2, sortPriority=1}
  FunctionOutput(None=Frac(None=@n5) {desc=Lerp Alpha}) {description=Lerp Alpha for UV 1 and 2, outputName=Crossfade Alpha, sortPriority=3}
 ]}
