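# Unlit alpha-composite material for the Firecracker back bling's smoke trail, turning mesh quads into looping sprites that spiral away from the component pivot and grow with age.
# Lifetime comes from time plus vertex alpha, driving travel, size and color, with MF_Backpack_Firecracker_Spiral supplying the motion and vertex red varying scale.
# A crossfaded two-texture 4x4 flipbook, offset and rotated through UV1, feeds color blending and erosion-based opacity with depth fade.
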Material /BRCosmetics/Accessories/FORT_Backpacks/Backpack_M_MED_Firecracker/FX/Materials/M_Backpack_Firecracker_SpriteTrail
graph {
 properties {
  bUsedWithNiagaraMeshParticles=true
  blendMode=BLEND_AlphaComposite
  pixelDepthOffsetMode=PDOM_Legacy
  shadingModel=MSM_Unlit
  usedShadingModels=Unlit
 }
 nodes [
  n1 = VectorParameter() {parameterName=ColorAlphaA2, defaultValue=<rgba 0.03 0.03 0.03 2>}
  n2 = AppendVector(A=@n1.RGB, B=@n1.RGB)
  n3 = VectorParameter() {parameterName=ColorAlphaB2, defaultValue=<rgba 0.1 0.0017 0 1>}
  n4 = AppendVector(A=@n3.RGB, B=@n3.RGB)
  n5 = TextureCoordinate() {coordinateIndex=1}
  n6 = Time()
  n7 = ScalarParameter() {defaultValue=0.5, parameterName=Flipbook Speed}
  n8 = MaterialFunctionCall!(
   UV=Rotator(Time=ComponentMask(None=@n5) {g=true}) {speed=6.28}
   XY Frames=Constant() {r=4}
   Animation Phase=Add(A=ComponentMask(None=@n5) {r=true}, B=Multiply(A=@n6, B=@n7))
  ) {materialFunction=&/MF_FlipbookUV}
  n9 = LinearInterpolate(
   A=AppendVector(
    A=TextureSample(UVs=@n8.UV 1) {texture=/Game/Effects/Niagara/MasterMaterials/Smoke/Textures/NEW/RT4x4_5_Tex}.R
    B=TextureSample(UVs=@n8.UV 1) {texture=/Game/Effects/Niagara/MasterMaterials/Smoke/Textures/NEW/RT4x4_4_Tex}.R
   )
   B=AppendVector(
    A=TextureSample(UVs=@n8.UV 2) {texture=&RT4x4_5_Tex}.R
    B=TextureSample(UVs=@n8.UV 2) {texture=&RT4x4_4_Tex}.R
   )
   Alpha=@n8.Crossfade Alpha
  )
  n10 = ComponentMask(None=@n9) {r=true}
  n11 = VectorParameter() {parameterName=ColorAlphaA1, defaultValue=<rgba 0.5 0.025 0 1>}
  n12 = AppendVector(A=@n11.RGB, B=@n11.RGB)
  n13 = VectorParameter() {parameterName=ColorAlphaB1, defaultValue=<rgba 13 3.1222 0.39 1>}
  n14 = AppendVector(A=@n13.RGB, B=@n13.RGB)
  n15 = ScalarParameter() {defaultValue=-45, parameterName=Move Distance}
  n16 = Power(Base=@n21, Exp=ScalarParameter() {defaultValue=1.6, parameterName=Move Exponent})
  n17 = Add(
   A=Multiply(A=Multiply(A=@n18.Length Vector, B=@n15), B=@n16)
   B=MaterialFunctionCall() {materialFunction=&ComponentPivotLocation}.Result
  )
  n18 = MaterialFunctionCall!(WorldPosition=@n17, WorldSpace Normal) {materialFunction=&/MF_Backpack_Firecracker_Spiral}
  n19 = Multiply(A=@n18.Rotation Speed, B=Time())
  n20 = ScalarParameter() {defaultValue=0.5, parameterName=Move Speed}
  n21 = Frac(None=Add(A=VertexColor().A, B=Multiply(A=@n19, B=@n20))) {desc=Lifetime}
  n22 = OneMinus(None=@n21)
  n23 = ScalarParameter() {defaultValue=2, parameterName=Color Blend Exponent}
  n24 = ScalarParameter() {defaultValue=0.2, parameterName=Gradient Center Point}
  n25 = MaterialFunctionCall!(
   Gradient=ComponentMask(None=@n9) {g=true}
   Erosion Amount=OneMinus(None=MaterialFunctionCall!(Gradient=@n21, Center Point=@n24) {materialFunction=/Game/Effects/Fort_Effects/MaterialFunctions/MF_GradientReflected}.Result)
   Softness=ScalarParameter() {defaultValue=0.4, parameterName=Erosion Softness}
  ) {materialFunction=/BRCosmetics/Athena/Effects/Particles/Skydiving/NS_Contrail_DarkAndLight/Functions/MF_Erosion}
  n26 = ScalarParameter() {defaultValue=0.6, parameterName=Opacity Multiplier, sliderMax=1}
  n27 = Multiply(
   A=LinearInterpolate(
    A=LinearInterpolate(A=@n2, B=@n4, Alpha=@n10)
    B=LinearInterpolate(A=@n12, B=@n14, Alpha=@n10)
    Alpha=Power(Base=@n22, Exp=@n23) {constExponent=6}
   )
   B=DepthFade(Opacity=Multiply(A=@n25.Default Range, B=@n26), FadeDistance=ScalarParameter() {defaultValue=4, parameterName=Depth Fade Distance})
  )
  ComponentMask(None=@n27) {a=true}
  n28 = Multiply(A=VertexColor().R, B=ScalarParameter() {defaultValue=0.3, parameterName=Sprite Scale Variation})
  n29 = ScalarParameter() {defaultValue=6, parameterName=Sprite Scale Birth}
  n30 = ScalarParameter() {defaultValue=44, parameterName=Sprite Scale Death}
  MaterialFunctionCall!(
   Center Location=Add(A=@n18.Spiral WorldPositionOffset, B=@n17)
   0-1 UVs
   XY scale=Multiply(A=OneMinus(None=@n28), B=LinearInterpolate(A=@n29, B=@n30, Alpha=@n16))
   OverridePivot
   "Normals (Optional)"
  ) {materialFunction=&Sprite}
  MaterialFunctionCall(Emissive=ComponentMask(None=@n27) {b=true, g=true, r=true}) {materialFunction=&MF_FX_EmissiveAdjustments}
 ]
}
