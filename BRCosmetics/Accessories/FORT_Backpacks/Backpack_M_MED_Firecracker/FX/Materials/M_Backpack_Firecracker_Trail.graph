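# Unlit alpha-composite trail material for the Firecracker backpack, set up for Niagara mesh particles and skeletal meshes.
# A skewed, panning noise texture feeds normalized erosion, driven by length along V and view angle against the spiral function's interpolated normal.
# Erosion blends two color pairs, shifting from dark tones to hot HDR colors at low V, and fades opacity along the trail.
# The spiral function also supplies world position offset.
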Material /BRCosmetics/Accessories/FORT_Backpacks/Backpack_M_MED_Firecracker/FX/Materials/M_Backpack_Firecracker_Trail
graph {
 properties {
  bTangentSpaceNormal=false
  bUsedWithMeshDeformer=true
  bUsedWithNiagaraMeshParticles=true
  bUsedWithSkeletalMesh=true
  blendMode=BLEND_AlphaComposite
  pixelDepthOffsetMode=PDOM_Legacy
  previewMesh=/BRCosmetics/Accessories/FORT_Backpacks/Backpack_M_MED_Firecracker/FX/M_MED_Firecracker_Trail
  shadingModel=MSM_Unlit
  usedShadingModels=Unlit
 }
 nodes [
  n1 = VectorParameter() {parameterName=ColorAlphaA2, defaultValue=<rgba 0.03 0.03 0.03 0.5>}
  n2 = AppendVector(A=@n1.RGB, B=@n1.RGB)
  n3 = VectorParameter() {parameterName=ColorAlphaB2, defaultValue=<rgba 0.1 0.0017 0 0.5>}
  n4 = AppendVector(A=@n3.RGB, B=@n3.RGB)
  n5 = MaterialFunctionCall(Float2=TextureCoordinate()) {materialFunction=&BreakOutFloat2Components}
  n6 = Add(A=@n5.R, B=Divide(A=@n5.G))
  n7 = TextureSample(UVs=Panner(Coordinate=AppendVector(A=@n6, B=@n5.G), Speed=Constant2Vector() {g=-1, r=-0.06})) {
   samplerType=SAMPLERTYPE_LinearGrayscale
   texture=/BRCosmetics/Accessories/FORT_Backpacks/Backpack_M_MED_Firecracker/FX/T_Backpack_Firecracker_Noise
  }
  n8 = OneMinus(None=@n5.G)
  n9 = MaterialFunctionCall!(WorldPosition, WorldSpace Normal) {materialFunction=&/MF_Backpack_Firecracker_Spiral}
  n10 = DotProduct(A=VertexInterpolator(VS=@n9.Spiral WorldSpace Normal).PS, B=CameraVectorWS())
  n11 = MaterialFunctionCall!(Gradient=@n7.R, Erosion Amount=OneMinus(None=Multiply(A=@n8, B=@n10))) {materialFunction=/BRCosmetics/Athena/Effects/Particles/Skydiving/NS_Contrail_DarkAndLight/Functions/MF_ErosionNormalized}
  n12 = VectorParameter() {parameterName=ColorAlphaA1, defaultValue=<rgba 0.5 0.025 0 0>}
  n13 = AppendVector(A=@n12.RGB, B=@n12.RGB)
  n14 = VectorParameter() {parameterName=ColorAlphaB1, defaultValue=<rgba 13 3.1222 0.39 0>}
  n15 = AppendVector(A=@n14.RGB, B=@n14.RGB)
  n16 = OneMinus(None=@n5.G)
  n17 = Multiply(
   A=LinearInterpolate(
    A=LinearInterpolate(A=@n2, B=@n4, Alpha=@n11.Result)
    B=LinearInterpolate(A=@n13, B=@n15, Alpha=@n11.Result)
    Alpha=Power(Base=@n16) {constExponent=9}
   )
   B=Multiply(A=SquareRoot(None=@n11.Result), B=@n16)
  )
  ComponentMask(None=@n17) {a=true}
  MaterialFunctionCall(Emissive=ComponentMask(None=@n17) {b=true, g=true, r=true}) {materialFunction=&MF_FX_EmissiveAdjustments}
  Reroute(None=@n9.Spiral WorldPositionOffset)
 ]
}
